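# Material function for the Athena Arid landscape rock layer, outputting full (Mat) and simple (MatSimple) material attributes.
# A world-aligned packed texture supplies MPC-tinted color, height (carried in Opacity) and roughness, with a flattened world-aligned normal.
# MatSimple shares height and normal but takes color from one XZ or YZ planar sample chosen by vertex normal, using explicit derivatives for mip selection.
# Height, normal and both low-cost colors are also separate outputs.
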MaterialFunction /FigmentEnvironment/Migration51/Athena/Environments/Landscape/MaterialFunctions/Arid/MF_Athena_Arid_Rock
graph {nodes [
  n1 = TextureObject() {
   samplerType=SAMPLERTYPE_LinearColor
   texture=/FigmentEnvironment/Migration51/Athena/Environments/Landscape/Textures/Arid/T_Athena_Arid_Rock_01_M
  }
  n2 = ScalarParameter() {defaultValue=2000, group=Arid_Rock, parameterName=Arid_Rock01_TexSize}
  n3 = FunctionInput() {bUsePreviewValueAsDefault=true, inputName=ProjContrast, inputType=FunctionInput_Scalar}
  n4 = MaterialFunctionCall(TextureObject=@n1, TextureSize=@n2, ProjectionTransitionContrast=@n3) {materialFunction=&WorldAlignedTexture}
  n5 = ComponentMask(None=@n4.XY Texture) {r=true}
  n6 = ComponentMask(None=CollectionParameter() {collection=/FigmentEnvironment/Migration51/Athena/Environments/Landscape/MPC/MPC_Landscape, parameterName=Arid_Rock_01_Tint}) {b=true, g=true, r=true}
  n7 = ComponentMask(None=@n4.XY Texture) {g=true}
  n8 = TextureObject() {samplerType=SAMPLERTYPE_Normal, texture=/FigmentEnvironment/Migration51/Environments/Sets/Arid/Textures/RockCliff_normal}
  n9 = OneMinus(None=@n3)
  n10 = FunctionOutput(None=LinearInterpolate(
    A=MaterialFunctionCall!(TextureObject=@n8, TextureSize=@n2, WorldSpace, Normal, ProjectionTransitionContrast=@n9, WorldPosition, Use High Quality Normals) {materialFunction=/FigmentEnvironment/Migration51/Athena/Environments/Materials/WorldAlignedNormalContrast}.XY Texture
    B=Constant3Vector() {constant=<rgba 0 0 1 0>}
    Alpha=ScalarParameter() {defaultValue=0.65, group=Arid_Rock, parameterName=Arid_Rock01_Flatten}
   )) {outputName=Normal}
  FunctionOutput(None=MakeMaterialAttributes!(
    BaseColor=Multiply(A=@n5, B=@n6)
    Metallic
    Specular
    Roughness=ComponentMask(None=@n4.XY Texture) {b=true}
    Anisotropy
    EmissiveColor
    Opacity=FunctionOutput(None=@n7) {outputName=Height}
    OpacityMask
    Normal=@n10
    Tangent
    WorldPositionOffset
    SubsurfaceColor
    ClearCoat
    ClearCoatRoughness
    AmbientOcclusion
    Refraction
    CustomizedUVs_0
    CustomizedUVs_1
    CustomizedUVs_2
    CustomizedUVs_3
    CustomizedUVs_4
    CustomizedUVs_5
    CustomizedUVs_6
    CustomizedUVs_7
    PixelDepthOffset
    ShadingModel
    Displacement
   )) {outputName=Mat}
  n11 = AppendVector(
   A=ScalarParameter() {defaultValue=10, group=Arid_Rock, parameterName=Arid_Rock_Tiling_X}
   B=ScalarParameter() {defaultValue=50, group=Arid_Rock, parameterName=Arid_Rock_Tiling_Y}
  )
  n12 = Divide(A=LandscapeLayerCoords() {mappingScale=1, mappingType=TCMT_XZ}, B=@n11)
  n13 = Divide(A=LandscapeLayerCoords() {mappingScale=1, mappingType=TCMT_YZ}, B=@n11)
  n14 = Abs(None=VertexNormalWS())
  n15 = Ceil(None=Subtract(A=ComponentMask(None=@n14) {r=true}, B=ComponentMask(None=@n14) {g=true}))
  n16 = DDX(Value=@n12)
  n17 = DDY(Value=@n12)
  n18 = DDX(Value=@n13)
  n19 = DDY(Value=@n13)
  n20 = Max(A=Abs(None=AppendVector(A=@n16, B=@n17)), B=Abs(None=AppendVector(A=@n18, B=@n19)))
  n21 = TextureSample!(
   UVs=LinearInterpolate(A=@n12, B=@n13, Alpha=Saturate(None=@n15))
   Tex
   "DDX(UVs)"=ComponentMask(None=@n20) {g=true, r=true}
   "DDY(UVs)"=ComponentMask(None=@n20) {a=true, b=true}
   Apply View MipBias
  ) {
   mipValueMode=TMVM_Derivative
   samplerSource=SSM_Wrap_WorldGroupSettings
   samplerType=SAMPLERTYPE_LinearColor
   texture=&T_Athena_Arid_Rock_01_M
  }
  FunctionOutput(None=@n21.R) {outputName=BaseColor_Plain}
  n22 = Multiply(A=@n21.R, B=@n6)
  FunctionOutput(None=MakeMaterialAttributes(BaseColor=FunctionOutput(None=@n22) {outputName=BaseColor_Low}, Opacity=@n7, Normal=@n10)) {outputName=MatSimple}
  ScalarParameter() {defaultValue=-10, group=Arid_Rock, parameterName=Arid_Rock_ProjContrast_1}
 ]}
