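# Landscape material function that tints a base texture from a topographic mask texture and also outputs peak (G), cavity (B) and red-channel masks.
# MPC_Landscape peak tints multiply in by blue and cavity tints by green, opposite to the output names, with the Arid input blending standard and arid versions.
# A fixed bluish tint then applies through GrassMask, upward vertex normal and sharpened inverse blue.
# The RockMask input is unconnected.
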MaterialFunction /FigmentEnvironment/Migration51/Athena/Environments/Landscape/MaterialFunctions/MF_TerrainTopoAdjustment
graph {nodes [
  FunctionInput() {bUsePreviewValueAsDefault=true, inputName=RockMask, inputType=FunctionInput_Scalar}
  n1 = TextureObject() {
   samplerType=SAMPLERTYPE_LinearColor
   texture=/FigmentEnvironment/Migration51/Athena/Environments/Landscape/Textures/T_Athena_Terrain_Topo_Mask
  }
  n2 = TextureSample(
   UVs=FunctionInput(Preview=LandscapeLayerCoords() {mappingScale=1525}) {bUsePreviewValueAsDefault=true, inputName=UV Coordinates, inputType=FunctionInput_Vector2}
   Tex=FunctionInput(Preview=@n1) {bUsePreviewValueAsDefault=true, inputName=Mask 2, inputType=FunctionInput_Texture2D}
  ) {
   samplerSource=SSM_Wrap_WorldGroupSettings
   samplerType=SAMPLERTYPE_LinearColor
   texture=&T_Athena_Terrain_Topo_Mask
  }
  FunctionOutput(None=@n2.G) {outputName=Peak_Mask}
  n3 = ComponentMask(None=CollectionParameter() {collection=/FigmentEnvironment/Migration51/Athena/Environments/Landscape/MPC/MPC_Landscape, parameterName=Terrain_Cav_Tint}) {b=true, g=true, r=true}
  n4 = CollectionParameter() {collection=&MPC_Landscape, parameterName=Terrain_Peak_Tint}
  n5 = FunctionInput() {inputName=BaseTexture}
  n6 = LinearInterpolate(A=@n5, B=Multiply(A=@n5, B=ComponentMask(None=@n4) {b=true, g=true, r=true}), Alpha=@n2.B)
  n7 = CollectionParameter() {collection=&MPC_Landscape, parameterName=Arid_Terrain_Peak_Tint}
  n8 = LinearInterpolate(A=@n5, B=Multiply(A=@n5, B=ComponentMask(None=@n7) {b=true, g=true, r=true}), Alpha=@n2.B)
  n9 = ComponentMask(None=CollectionParameter() {collection=&MPC_Landscape, parameterName=Arid_Terrain_Cav_Tint}) {b=true, g=true, r=true}
  n10 = LinearInterpolate(
   A=LinearInterpolate(A=@n6, B=Multiply(A=@n6, B=@n3), Alpha=@n2.G)
   B=LinearInterpolate(A=@n8, B=Multiply(A=@n8, B=@n9), Alpha=@n2.G)
   Alpha=FunctionInput() {bUsePreviewValueAsDefault=true, inputName=Arid}
  )
  n11 = Constant3Vector() {constant=<rgba 0.7616 0.9404 1.1 1>}
  n12 = Multiply(
   A=Power(Base=OneMinus(None=@n2.B), Exp=Constant() {r=4})
   B=MaterialFunctionCall(Float3=VertexNormalWS()) {materialFunction=&BreakOutFloat3Components}.B
  )
  n13 = FunctionInput() {bUsePreviewValueAsDefault=true, inputName=GrassMask, inputType=FunctionInput_Scalar}
  FunctionOutput(None=LinearInterpolate(A=@n10, B=Multiply(A=@n10, B=@n11), Alpha=Multiply(A=@n12, B=@n13)))
  FunctionOutput(None=@n2.B) {outputName=Cavity_Mask}
  FunctionOutput(None=@n2.R) {outputName=RedChannel}
 ]}
