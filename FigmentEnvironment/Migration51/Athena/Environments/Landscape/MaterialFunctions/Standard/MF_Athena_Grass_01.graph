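# Material function that builds the grass layer of the Athena landscape material, outputting full material attributes plus a plain base color.
# MF_Landscape_GrassDistanceBlend supplies grass color, also wired to Opacity, and an alpha that is raised to a power to scale specular.
# A grasslands normal map is projected by world XY through a shared wrap sampler with adjustable scale, and roughness is a scalar parameter.
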MaterialFunction /FigmentEnvironment/Migration51/Athena/Environments/Landscape/MaterialFunctions/Standard/MF_Athena_Grass_01
graph {nodes [
  n1 = MaterialFunctionCall!(UseVertexInterp) {materialFunction=/FigmentEnvironment/Migration51/Athena/Environments/Landscape/MaterialFunctions/MF_Landscape_GrassDistanceBlend}
  n2 = ScalarParameter() {defaultValue=0.15, group=Grass, parameterName=GrassSpecAmount}
  n3 = Power(Base=@n1.Grass A, Exp=ScalarParameter() {defaultValue=2, group=Grass, parameterName=GrassSpecPower})
  n4 = WorldPosition() {worldPositionShaderOffset=WPT_ExcludeAllShaderOffsets}
  n5 = TextureSample(UVs=Divide(A=ComponentMask(None=@n4.XYZ) {g=true, r=true}, B=ScalarParameter() {defaultValue=700, parameterName=GrassNearScale})) {
   samplerSource=SSM_Wrap_WorldGroupSettings
   samplerType=SAMPLERTYPE_Normal
   texture=/FigmentEnvironment/Migration51/Packages/DS_Fortnite_Terrain_NoLOD/Textures/T_Grasslands_N
  }
  FunctionOutput(None=MakeMaterialAttributes!(
    BaseColor=@n1.Grass RGB
    Metallic
    Specular=LinearInterpolate(B=@n2, Alpha=@n3)
    Roughness=ScalarParameter() {defaultValue=0.9, parameterName=Grass_Roughness}
    Anisotropy
    EmissiveColor
    Opacity=@n1.Grass RGB
    OpacityMask
    Normal=@n5.RGB
    Tangent
    WorldPositionOffset
    SubsurfaceColor
    ClearCoat
    ClearCoatRoughness
    AmbientOcclusion
    Refraction
    CustomizedUVs_0
    CustomizedUVs_1
    CustomizedUVs_2
    CustomizedUVs_3
    CustomizedUVs_4
    CustomizedUVs_5
    CustomizedUVs_6
    CustomizedUVs_7
    PixelDepthOffset
    ShadingModel
    Displacement
   ))
  FunctionOutput(None=@n1.Grass RGB) {outputName=BaseColor_Plain}
 ]}
