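# Landscape material function that builds the Athena mud layer as material attributes from one mud texture sampled on input UVs.
# Texture color is multiplied by MudTint from MPC_Landscape, red drives specular between min and max parameters, alpha feeds Opacity, and roughness is fully rough.
# A second output, BaseColor_Plain, exposes untinted texture color.
# A gravel branch with desaturation, HeightLerp and tint blend is left unconnected to either output.
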MaterialFunction /FigmentEnvironment/Migration51/Athena/Environments/Landscape/MaterialFunctions/Standard/MF_Athena_Mud_01
graph {nodes [
  Constant3Vector() {constant=<rgba 0 0 1 0>}
  Constant() {r=0.2}
  n1 = FunctionInput(Preview=TextureCoordinate()) {bUsePreviewValueAsDefault=true, inputType=FunctionInput_Vector2}
  n2 = TextureSampleParameter2D(UVs=@n1) {
   group=Mud
   parameterName=Mud
   samplerSource=SSM_Wrap_WorldGroupSettings
   texture=/FigmentEnvironment/Migration51/Athena/Environments/Landscape/Textures/Arid/T_Athena_Arid_MudVariant_01
  }
  n3 = ComponentMask(None=CollectionParameter() {collection=/FigmentEnvironment/Migration51/Athena/Environments/Landscape/MPC/MPC_Landscape, parameterName=MudTint}) {b=true, g=true, r=true}
  n4 = ScalarParameter() {defaultValue=0.2, group=Road, parameterName=Mud_SpecMin}
  n5 = ScalarParameter() {defaultValue=0.5, group=Road, parameterName=Mud_SpecMax}
  FunctionOutput(None=MakeMaterialAttributes!(
    BaseColor=Multiply(A=@n2.RGB, B=@n3)
    Metallic
    Specular=LinearInterpolate(A=@n4, B=@n5, Alpha=@n2.R)
    Roughness=Constant() {r=1}
    Anisotropy
    EmissiveColor
    Opacity=@n2.A
    OpacityMask
    Normal
    Tangent
    WorldPositionOffset
    SubsurfaceColor
    ClearCoat
    ClearCoatRoughness
    AmbientOcclusion
    Refraction
    CustomizedUVs_0
    CustomizedUVs_1
    CustomizedUVs_2
    CustomizedUVs_3
    CustomizedUVs_4
    CustomizedUVs_5
    CustomizedUVs_6
    CustomizedUVs_7
    PixelDepthOffset
    ShadingModel
    Displacement
   ))
  FunctionOutput(None=@n2.RGB) {outputName=BaseColor_Plain}
  n6 = TextureSampleParameter2D(UVs=@n1) {
   group=Gravel
   parameterName=Gravel
   samplerSource=SSM_Wrap_WorldGroupSettings
   texture=/FigmentEnvironment/Migration51/Packages/DS_Fortnight/WhiteRockCustom_D
  }
  n7 = ComponentMask(None=@n6.RGB) {b=true, g=true, r=true}
  LinearInterpolate(A=@n7)
  n8 = Desaturation(None=@n7, Fraction=ScalarParameter() {parameterName=Desat_test})
  n9 = VectorParameter() {group=Mud, parameterName=MudTint, defaultValue=<rgba 0.75 0.6705 0.582 0>}
  MaterialFunctionCall!(
   Material=MakeMaterialAttributes(BaseColor=Multiply(A=@n8, B=@n9.RGB), Roughness=Constant() {r=1})
   Tint=VectorParameter() {parameterName=MudTintDarken, defaultValue=<rgba 0 0 0.2 0>}.B
   TintMask=MaterialFunctionCall(Transition Phase=ScalarParameter() {defaultValue=2, parameterName=Test}, Height Texture=OneMinus(None=@n6.A), Contrast=Constant()) {materialFunction=&HeightLerp}.Lerp Alpha No Contrast
  ) {materialFunction=/Engine/Functions/MaterialLayerFunctions/MatLayerBlend_Tint}
 ]}
