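# Arid terrain master material for static and spline meshes, apparently a backup, matching Athena landscape layers.
# Vertex colors drive height-lerped sand, mud, road, gravel and grass; Use_Cliffs swaps mud and road for rock blended by alpha.
# IsDesertRoad inverts the mud mask, applies the DesertRoad function and adjusts color by terrain topo mask.
# Normals are re-oriented by surface axis for world projection, then get a distance-faded macro normal, simplified on low-end platforms.
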Material /Figment_S05/Migration51/Athena/Environments/Landscape/Material/Arid/M_Athena_AridSM_MasterBackup
graph {
 properties {
  bUseMaterialAttributes=true
  bUsedWithSplineMeshes=true
  pixelDepthOffsetMode=PDOM_Legacy
  refractionMethod=RM_IndexOfRefraction
  usedShadingModels=Default Lit
 }
 nodes [
  Constant() {r=1}
  LandscapeLayerCoords() {mappingScale=1525}
  n1 = MaterialFunctionCall!(
   In=MaterialFunctionCall!(ProjContrast) {materialFunction=/FigmentEnvironment/Migration51/Athena/Environments/Landscape/MaterialFunctions/Arid/MF_Athena_Arid_Rock}.Mat
   Color=MaterialFunctionCall() {materialFunction=/FigmentEnvironment/Migration51/Athena/Environments/Landscape/MaterialFunctions/MF_Athena_SedimentGradient}.Result
  ) {materialFunction=/FigmentEnvironment/Migration51/Athena/Environments/Landscape/MaterialFunctions/MF_Athena_Material_BaseMultiply}
  n2 = CollectionParameter() {collection=/FigmentEnvironment/Migration51/Athena/Environments/Landscape/MPC/MPC_Landscape, parameterName=Arid_Sand_TexSize}
  n3 = Divide(A=WorldPosition() {worldPositionShaderOffset=WPT_ExcludeAllShaderOffsets}.XYZ, B=Multiply(A=Abs(None=@n2)) {constB=-1})
  n4 = MaterialFunctionCall(
   In=MaterialFunctionCall!(UseRoughness=StaticBool(), UVs=ComponentMask(None=@n3) {g=true, r=true}) {materialFunction=/FigmentEnvironment/Migration51/Athena/Environments/Landscape/MaterialFunctions/Arid/MF_Athena_Arid_Sand}.Simple_Material
   Color=MaterialFunctionCall() {materialFunction=/FigmentEnvironment/Migration51/Athena/Environments/Landscape/MaterialFunctions/MF_Athena_ZSandColor}.Result
  ) {materialFunction=&MF_Athena_Material_BaseMultiply}
  n5 = Abs(None=ScalarParameter() {defaultValue=1024, group=MasterControl, parameterName=Texturesize_Z})
  n6 = ComponentMask(None=Divide(A=WorldPosition() {worldPositionShaderOffset=WPT_ExcludeAllShaderOffsets}.XYZ, B=Multiply(A=@n5) {constB=-1})) {g=true, r=true}
  n7 = MaterialFunctionCall(
   In=MaterialFunctionCall(
    In=MaterialFunctionCall!(In=@n6) {materialFunction=/FigmentEnvironment/Migration51/Athena/Environments/Landscape/MaterialFunctions/Standard/MF_Athena_Mud_01}.Result
    Color=CollectionParameter() {collection=&MPC_Landscape, parameterName=Arid_Mud_Tint}
   ) {materialFunction=&MF_Athena_Material_BaseMultiply}.Result
   Color=MaterialFunctionCall() {materialFunction=&MF_Athena_ZSandColor}.Result
  ) {materialFunction=&MF_Athena_Material_BaseMultiply}
  n8 = VertexColor()
  n9 = Clamp(None=StaticSwitchParameter(True=OneMinus(None=@n8.A), False=@n8.A) {parameterName=IsDesertRoad})
  n10 = MaterialFunctionCall!(In) {materialFunction=/Figment_S05/Migration51/Athena/Environments/Landscape/MaterialFunctions/Arid/MF_Athena_Arid_SandWavy}
  n11 = ScalarParameter() {defaultValue=1, parameterName=Mud_Blend_Contrast}
  n12 = VertexColor()
  n13 = MaterialFunctionCall!(UseSimple=StaticBool(), In=@n6) {materialFunction=/FigmentEnvironment/Migration51/Athena/Environments/Landscape/MaterialFunctions/Standard/MF_Athena_Road_01}
  n14 = MaterialFunctionCall(
   Transition Phase=Power(Base=@n12.R, Exp=ScalarParameter() {defaultValue=2, parameterName=RoadMud_Trans})
   Height Texture=MaterialFunctionCall!(Mat attributes=@n13.Result) {materialFunction=/Engine/Functions/Engine_MaterialFunctions02/MatLayerBlend_BreakOpacity}.Result
   Contrast=ScalarParameter() {defaultValue=2, parameterName=RoadMud_Contrast}
  ) {materialFunction=&HeightLerp}
  n15 = OneMinus(None=@n14.Alpha)
  n16 = MaterialFunctionCall!(
   Base Material=MaterialFunctionCall!(Material=@n7.Result, Tint=@n15, TintMask=Constant() {r=1}) {materialFunction=/Engine/Functions/MaterialLayerFunctions/MatLayerBlend_Tint}.Blended Material
   Top Material=@n13.Result
   Alpha=@n14.Alpha
  ) {materialFunction=/Engine/Functions/MaterialLayerFunctions/MatLayerBlend_Standard}
  n17 = Multiply(A=@n12.R, B=ScalarParameter() {defaultValue=1.5, parameterName=RoadMud_Trans_1})
  n18 = MaterialFunctionCall(
   Base Material=MaterialFunctionCall(
    Base Material=@n4.Result
    Top Material=@n7.Result
    Alpha=MaterialFunctionCall(Transition Phase=@n9, Height Texture=@n10.Height, Contrast=@n11) {desc=Mud, materialFunction=&HeightLerp}.Alpha
   ) {materialFunction=&MatLayerBlend_Standard}.Blended Material
   Top Material=@n16.Blended Material
   Alpha=MaterialFunctionCall(Transition Phase=@n17, Height Texture=@n10.Height, Contrast=ScalarParameter() {defaultValue=1, parameterName=Road_Blend_Contrast}) {desc=Road, materialFunction=&HeightLerp}.Alpha
  ) {materialFunction=&MatLayerBlend_Standard}
  n19 = MaterialFunctionCall!(In=@n6) {materialFunction=/FigmentEnvironment/Migration51/Athena/Environments/Landscape/MaterialFunctions/Standard/MF_Athena_Gravel_01}
  n20 = Multiply(
   A=@n19.Plain_BaseColor
   B=ComponentMask(None=CollectionParameter() {collection=&MPC_Landscape, parameterName=Arid_Gravel_Tint}) {b=true, g=true, r=true}
  )
  n21 = VertexColor()
  n22 = ScalarParameter() {defaultValue=2, parameterName=Gravel_Blend_Contrast}
  n23 = MaterialFunctionCall() {materialFunction=/FigmentEnvironment/Migration51/Athena/Environments/Landscape/MaterialFunctions/Standard/MF_Athena_Grass_01}
  n24 = MaterialFunctionCall(
   Base Material=MaterialFunctionCall(
    Base Material=StaticSwitchParameter(True=@n4.Result, False=@n18.Blended Material) {defaultValue=true, parameterName=Use_Cliffs}
    Top Material=MaterialFunctionCall!(Mat_In=@n19.Result, BaseColor=@n20) {materialFunction=/FigmentEnvironment/Migration51/Athena/Environments/Landscape/MaterialFunctions/MF_Athena_ReplaceBaseColor}.Result
    Alpha=MaterialFunctionCall(Transition Phase=@n21.B, Height Texture=@n19.Height, Contrast=@n22) {desc=Gravel, materialFunction=&HeightLerp}.Alpha
   ) {materialFunction=&MatLayerBlend_Standard}.Blended Material
   Top Material=@n23.Result
   Alpha=MaterialFunctionCall(
    Transition Phase=VertexColor().G
    Height Texture=MaterialFunctionCall(Mat attributes=@n23.Result) {materialFunction=&MatLayerBlend_BreakOpacity}.Result
    Contrast=ScalarParameter() {parameterName=Grass_Blend_Contrast}
   ) {materialFunction=&HeightLerp}.Alpha
  ) {materialFunction=&MatLayerBlend_Standard}
  n25 = MaterialFunctionCall(
   Transition Phase=OneMinus(None=VertexColor().A)
   Height Texture=Constant() {r=0.5}
   Contrast=ScalarParameter() {defaultValue=0.5, parameterName=Contrasttest}
  ) {materialFunction=&HeightLerp}
  n26 = BreakMaterialAttributes!(Attr=StaticSwitchParameter(
    True=MaterialFunctionCall(Base Material=@n1.Result, Top Material=@n24.Blended Material, Alpha=@n25.Alpha) {materialFunction=&MatLayerBlend_Standard}.Blended Material
    False=@n24.Blended Material
   ) {defaultValue=true, parameterName=Use_Cliffs})
  n27 = MakeMaterialAttributes!(
   BaseColor=@n26.BaseColor
   Metallic
   Specular=Multiply(A=@n26.BaseColor, B=ScalarParameter() {defaultValue=0.5, parameterName=Spec_Amount})
   Roughness=@n26.BaseColor
   Anisotropy
   EmissiveColor
   Opacity=@n26.BaseColor
   OpacityMask
   Normal=@n26.BaseColor
   Tangent
   WorldPositionOffset
   SubsurfaceColor
   ClearCoat
   ClearCoatRoughness
   AmbientOcclusion=@n26.BaseColor
   Refraction
   CustomizedUVs_0
   CustomizedUVs_1
   CustomizedUVs_2
   CustomizedUVs_3
   CustomizedUVs_4
   CustomizedUVs_5
   CustomizedUVs_6
   CustomizedUVs_7
   PixelDepthOffset
   ShadingModel
   Displacement
  )
  n28 = MaterialFunctionCall!(In=@n27) {materialFunction=/Figment_S05/Migration51/Environments/Sets/Arid/Materials/DesertRoad}
  n29 = TextureObject() {
   samplerType=SAMPLERTYPE_LinearColor
   texture=/FigmentEnvironment/Migration51/Athena/Environments/Landscape/Textures/T_Athena_Terrain_Topo_Mask
  }
  n30 = ComponentMask(None=WorldPosition().XYZ) {g=true, r=true}
  n31 = ScalarParameter() {defaultValue=306200, group=TerrainColorMatch, parameterName=TerrainSize}
  n32 = Divide(A=@n31)
  n33 = AppendVector(A=@n32, B=@n32)
  n34 = VertexInterpolator(VS=Divide(A=Subtract(A=@n30, B=@n33), B=@n31))
  n35 = Constant() {r=1}
  n36 = MaterialFunctionCall!(Material=@n28.Result) {materialFunction=/Engine/Functions/MaterialLayerFunctions/MatLayerBlend_BreakBaseColor}
  n37 = StaticSwitchParameter(
   True=MaterialFunctionCall(
    Mat_In=@n28.Result
    BaseColor=MaterialFunctionCall!(Mask 2=@n29, UV Coordinates=@n34.PS, RockMask, GrassMask=@n35, Arid, BaseTexture=@n36.BaseColor) {materialFunction=/FigmentEnvironment/Migration51/Athena/Environments/Landscape/MaterialFunctions/MF_TerrainTopoAdjustment}.Result
   ) {materialFunction=&MF_Athena_ReplaceBaseColor}.Result
   False=@n27
  ) {parameterName=IsDesertRoad}
  n38 = MaterialFunctionCall!(Material=@n37) {materialFunction=/Engine/Functions/MaterialLayerFunctions/MatLayerBlend_BreakNormal}
  n39 = MaterialFunctionCall(Float3=@n38.Normal) {materialFunction=&BreakOutFloat3Components}
  n40 = Multiply(A=Transform(None=Constant3Vector() {constant=<rgba 0 0 1 0>}), B=Constant() {r=1})
  n41 = ComponentMask(None=@n40) {g=true}
  n42 = AppendVector(A=AppendVector(A=Multiply(A=@n39.R) {constB=-1}, B=Multiply(A=@n39.B, B=@n41)), B=Multiply(A=@n39.G) {constB=-1})
  n43 = ComponentMask(None=@n40) {r=true}
  n44 = AppendVector(A=AppendVector(A=Multiply(A=@n43, B=@n39.B), B=Multiply(A=@n39.R) {constB=-1}), B=Multiply(A=@n39.G) {constB=-1})
  n45 = Constant()
  n46 = MaterialFunctionCall(In=Abs(None=@n43), Contrast=@n45) {materialFunction=&CheapContrast}
  n47 = ComponentMask(None=Multiply(A=@n38.Normal, B=Constant3Vector() {constant=<rgba 1 1 1 0>})) {g=true, r=true}
  n48 = ComponentMask(None=@n40) {b=true}
  n49 = Multiply(A=@n39.B, B=@n48)
  n50 = Abs(None=@n48)
  n51 = Normalize(VectorInput=LinearInterpolate(
    A=LinearInterpolate(A=@n42, B=@n44, Alpha=@n46.Result)
    B=AppendVector(A=@n47, B=@n49)
    Alpha=MaterialFunctionCall(In=@n50, Contrast=@n45) {materialFunction=&CheapContrast}.Result
   ))
  n52 = MaterialFunctionCall!(
   Material=@n37
   Normal=MaterialFunctionCall(
    Normal=Transform(None=@n51) {transformSourceType=TRANSFORMSOURCE_World, transformType=TRANSFORM_Tangent}
    Flatness=ScalarParameter() {group=MasterControl, parameterName=NormalFlatten}
   ) {materialFunction=&FlattenNormal}.Result
  ) {materialFunction=/Engine/Functions/MaterialLayerFunctions/MatLayerBlend_ReplaceNormals}
  n53 = WorldPosition() {worldPositionShaderOffset=WPT_ExcludeAllShaderOffsets}
  n54 = TextureSample(UVs=Divide(A=ComponentMask(None=@n53.XYZ) {g=true, r=true}, B=ScalarParameter() {defaultValue=25000, parameterName=MacroNormalScale})) {
   samplerType=SAMPLERTYPE_Normal
   texture=/FigmentEnvironment/Migration51/ContentCreationTools/TextureCreation/Textures/T_BPCreated_MacroNormal_01
  }
  n55 = ScalarParameter() {defaultValue=0.625, parameterName=MacroNormalIntensity}
  n56 = Constant() {r=0.25}
  n57 = Constant() {r=0.9}
  n58 = FeatureLevelSwitch(Default=@n56, ES2=@n57, ES3_1=@n57, SM4_REMOVED=@n56)
  n59 = Multiply(
   A=LinearInterpolate(
    A=Constant3Vector() {constant=<rgba 0 0 1 0>}
    B=@n54.RGB
    Alpha=MaterialFunctionCall() {materialFunction=/FigmentEnvironment/Migration51/Athena/Environments/Landscape/MaterialFunctions/MF_TerrainDistanceFade}.Result
   )
   B=MaterialFunctionCall(X=@n55, Y=@n55, Z=QualitySwitch(Default=@n58, Low=@n57)) {materialFunction=&MakeFloat3}.Result
  )
  n60 = Add(
   A=Multiply(
    A=MaterialFunctionCall(Material=@n52.Result) {materialFunction=&MatLayerBlend_BreakNormal}.Normal
    B=Constant3Vector() {constant=<rgba 1 1 0.5 0>}
   )
   B=@n59
  )
  n61 = QualitySwitch(Default=@n60, Low=@n59, High=@n60, Medium=@n60)
  MaterialFunctionCall(Material=@n52.Result, Normal=FeatureLevelSwitch(Default=@n61, ES2=@n59, ES3_1=@n59, SM4_REMOVED=@n61, SM5=@n60, SM6=@n60)) {materialFunction=&MatLayerBlend_ReplaceNormals}
  Multiply(A=@n15, B=ScalarParameter())
  Multiply(A=@n32)
  Multiply(A=@n32) {constB=-1}
  Multiply(
   A=@n10.Height
   B=MaterialFunctionCall(Mat attributes=@n16.Blended Material) {materialFunction=&MatLayerBlend_BreakOpacity}.Result
  )
  ScalarParameter() {defaultValue=1500, group=MasterControl, parameterName=Texturesize_Z_Sand}
  StaticBool()
  Subtract(A=OneMinus(None=@n17), B=@n14.Alpha)
  VertexNormalWS()
 ]
}
